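# Idle effect for the FearCatch pickaxe, drawing persistent green sprites, a mesh and a light on the axe.
# One emitter bursts five spawn groups, writing the group to VisibilityTag and switching color, size and socket by it.
# Particles follow skeletal mesh sockets each update, with one group spread by ShapeLocation and sprite alignment taken from socket rotation.
# Some colors vary by time of day; two other emitters are empty.
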NiagaraSystem /BRCosmetics/Weapons/FORT_Melee/Pickaxe_FearCatch/FX/NS_Pickaxe_FearCatch_Idle
dependencies {
 usedDataInterfaces [
  &NiagaraDataInterfaceArrayFloat2
  &NiagaraDataInterfaceSkeletalMesh
  &NiagaraDataInterfaceArrayInt32
 ]
 usedDynamicInputs [&AddVectorToPosition]
 usedModules [
  &SystemState
  &self:SystemSpawnScript.SystemScriptSource.SystemScriptGraph.NiagaraNodeAssignment_0.SetVariables_#1
  &SampleSkeletalMesh
 ]
}
emitters {
 BlowingParticles {
  emitterUpdateScript [
   &EmitterState(
    Inactive Response=::NewEnumerator1
    Life Cycle Mode=::NewEnumerator1
    Loop Behavior=::NewEnumerator1
    Loop Duration Mode=::NewEnumerator1
    UseLoopDelay=false
   )
   &SpawnBurst_Instantaneous(Spawn Count=3, Use Spawn Probability=false, UseLoopCountLimit=false)
   &SpawnBurst_Instantaneous(Spawn Group=1, Use Spawn Probability=false, UseLoopCountLimit=false) {module=SpawnBurst_Instantaneous001}
   &SpawnBurst_Instantaneous(Spawn Group=2, Use Spawn Probability=false, UseLoopCountLimit=false) {module=SpawnBurst_Instantaneous002}
   &SpawnBurst_Instantaneous(Spawn Count=9, Spawn Group=3, Use Spawn Probability=false, UseLoopCountLimit=false) {module=SpawnBurst_Instantaneous003}
   &SpawnBurst_Instantaneous(Spawn Group=4, Use Spawn Probability=false, UseLoopCountLimit=false) {module=SpawnBurst_Instantaneous004}
  ]
  particleSpawnScript [
   /SetVariables(Particles.VisibilityTag=&SpawnGroup()) {isSetParameters=true}
   &PartitionParticles(Number Of Partitions=&MaskIntBySpawnGroup(Masked Int=1, Passthrough Int=3, Spawn Group Mask=3))
   &InitializeParticle(
    Color=&MergeHelper_ColorSwitch(
     0=&FNTimeOfDayColor(
      Night Color=<rgba 0 6 0.1341 0.0783>
      Morning Color=<rgba 0 3 0.067 0.0609>
      Day Color=<rgba 0 3 0.067 0.0609>
      Evening Color=<rgba 0 3 0.067 0.0609>
     )
     1=&FNTimeOfDayColor(Night Color=<rgba 0 3 0.067 0.25>, Morning Color=<rgba 0 3 0.067 0.4>, Day Color=<rgba 0 3 0.067 0.4>, Evening Color=<rgba 0 3 0.067 0.4>)
     2=<rgba 0 3 0.067 0.4>
     3=&FNTimeOfDayColor(
      Night Color=<rgba 0 6 0.1341 0.0348>
      Morning Color=<rgba 0 3 0.067 0.0609>
      Day Color=<rgba 0 3 0.067 0.0609>
      Evening Color=<rgba 0 3 0.067 0.0609>
     )
     Index=&SpawnGroup()
     Num=::NewEnumerator2
    )
    Mass Max=1.5
    Mass Mode=::NewEnumerator3
    Sprite Size=&MergeHelper_Vector2Switch(
     0=&SelectVector2DFromArray(
      Direct Array Index=$Particles.PartitionParticles.Partition
      Vector 2D Selection Array=DataInterface() {propertyValues {FloatData [<6 25>, <5 14.4974>, <6.9507 10.7853>]}}
     )
     1=<10 10>
     2=<-0.4034 0>
     3=<15 15>
     4=?
     Index=&SpawnGroup()
     Num=::NewEnumerator2
    )
    Sprite Size Mode=::NewEnumerator4
    UsePositionOffset=false
    Write Material Random=true
   )
   &CameraOffset(Camera Offset Amount=10)
  ]
  particleUpdateScript [
   &SkeletalMeshLocation(
    Bone / Socket Index=&MergeHelper_IntSwitch(
     0=&SelectIntFromArray(Direct Array Index=$Particles.PartitionParticles.Partition, Int Selection Array=DataInterface() {propertyValues {IntData [0,1,2]}})
     1=3
     2=0
     3=0
     4=4
     5=?
     Index=&SpawnGroup()
     Num=::NewEnumerator3
    )
    MaskSpawnGroup=false
    Mesh Sampling Type=::NewEnumerator5
    Orientation Sampling=::NewEnumerator1
    Skeletal Mesh=DataInterface() {propertyValues {
      DefaultMesh=null
      ExcludeBoneName=root
      OptionalComponentClass=null
      PreviewMesh=/BRCosmetics/Weapons/FORT_Melee/Pickaxe_FearCatch/Meshes/FearCatch_Axe
      SkinningMode=SkinOnTheFly
      SoftSourceActor=null
      SourceMode=Default
      UvSetIndex=0
      WholeMeshLOD=-1
      bExcludeBone=false
      bReadDeformedGeometry=true
      bRequireCurrentFrameData=false
      ComponentTags []
      FilteredBones []
      FilteredSockets [
       Idle_Melee_TwoHanded_HandleSocket_1
       Idle_Melee_TwoHanded_HandleSocket_2
       Idle_Melee_TwoHanded_HandleSocket_3
       IdleMelee_TwoHanded_HandleSocket_1
       Ring1Melee_TwoHanded_HandleSocket_2
      ]
      MeshUserParameter {parameter {name=None, typeDefHandle {registeredTypeIndex=1}, varData []}}
      SamplingRegions []
     }}
    Socket Sampling Mode=::NewEnumerator5
   )
   &ShapeLocation(
    MaskSpawnGroup=true
    Offset=<0 -18.447 53>
    Offset Coordinate Space=::Local
    Offset Mode=::NewEnumerator0
    Ring / Disc Distribution Mode:ENiagaraSphereDistributionMode=::NewEnumerator3
    Ring Radius=5
    Rotation Mode=::NewEnumerator2
    Shape Primitive=::NewEnumerator4
    Spawn Group Mask=3
    Uniform Spiral Amount=14.3188
    Uniform Spiral Falloff=0
    UseOwnerScale=true
    Yaw / Pitch / Roll=<0 0 0.75>
   )
   &ParticleState(Kill Particles When Lifetime Has Elapsed=false)
   &FN_Light_Attributes(Color=<rgba 0.413 3 0 1>, Create Color Attribute=true, Create Radius Attribute=true, Enable Visibility Tag=true, Radius=60)
   &SolveForcesAndVelocity(Clamp Velocity=false, Limit Acceleration=false)
   /SetVariables(Particles.SpriteAlignNoCam:Vector3f=&RotateVectorByQuaternion(Vector To Rotate=<0 1 0>, Quaternion=$Particles.SkeletalMeshLocation.SampledRotation)) {isSetParameters=true}
  ]
  rendererClasses [
   &NiagaraSpriteRendererProperties
   &NiagaraMeshRendererProperties
   &NiagaraLightRendererProperties
  ]
  renderers [
   {rendererClass=&NiagaraSpriteRendererProperties}
   {rendererClass=&NiagaraSpriteRendererProperties, rendererIndex=1}
   {rendererClass=&NiagaraMeshRendererProperties, rendererIndex=2}
   {rendererClass=&NiagaraSpriteRendererProperties, rendererIndex=3}
   {rendererClass=&NiagaraLightRendererProperties, rendererIndex=4}
  ]
 }
 Empty {rendererClasses [], renderers []}
 Empty001 {rendererClasses [], renderers []}
}
summary {
 systemName=NS_Pickaxe_FearCatch_Idle
 emitters [
  {
   emitterName=BlowingParticles
   rendererClasses [
    &NiagaraSpriteRendererProperties
    &NiagaraMeshRendererProperties
    &NiagaraLightRendererProperties
   ]
  }
  {emitterName=Empty, rendererClasses []}
  {emitterName=Empty001, rendererClasses []}
 ]
 userVariables []
}
