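# Niagara swing trail effect for the FearCatch pickaxe, drawing short-lived green glow and neutral metal anim trails.
# MainTrail feeds three ribbon renderers and a sprite renderer, with glow and metal colors faded over system age.
# Supporting_Trail adds a narrower trail from a second bone or socket pair through two ribbon renderers.
# Both emitters spawn via anim-trail modules, use time-of-day color selection, and fade alpha over particle life.
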NiagaraSystem /BRCosmetics/Weapons/FORT_Melee/Pickaxe_FearCatch/FX/NS_Pickaxe_FearCatch_Swing
dependencies {
 usedDataInterfaces [&NiagaraDataInterfaceColorCurve, &NiagaraDataInterfaceCurve, &NiagaraDataInterfaceSocketReader]
 usedDynamicInputs []
 usedModules [
  &self:SystemSpawnScript.SystemScriptSource.SystemScriptGraph.NiagaraNodeAssignment_0.SetVariables_#1
  &SystemState
 ]
}
emitters {
 MainTrail {
  emitterUpdateScript [
   &EmitterState()
   &FN_SpawnAnimTrail(Spawn Interval=0) {module=FN_SpawnAnimTrail001}
  ]
  particleSpawnScript [
   &FN_InitializeAnimTrailBoneIndecies()
   &FN_InitializeAnimTrail(AnimTrailWidth=1.2, Color=<rgba 0 3 0.5926 0.8261>, Lifetime=0.2)
   /SetVariables(Particles.float=2.569) {isSetParameters=true}
  ]
  particleUpdateScript [
   /SetVariables(
    Particles.GlowColor=&Multiply_LinearColorByFloat(
     Linear Color=&ColorFromCurve(ColorCurve=DataInterface() {propertyValues {
        AlphaCurve {keys [<at 0.201 0.464 RCIM_Cubic>,<at 0.7278 0>]}
        BlueCurve {keys [<at 0.4889 0.067 RCIM_Cubic>]}
        GreenCurve {keys [<at 0.4889 3 RCIM_Cubic>]}
        RedCurve {keys [<at 0.4889 0 RCIM_Cubic>]}
       }})
     Float=&FloatFromCurve(FloatCurve=DataInterface() {propertyValues {Curve {keys [<at 0 1>,<at 1 0>]}}}, CurveIndex=$Engine.System.Age)
    )
    Particles.MetalColor:LinearColor=&Multiply_LinearColorByFloat(
     Linear Color=&FNTimeOfDayColor(
      Night Color=<rgba 0.1143 0.1397 0.224 1>
      Morning Color=<rgba 0.3073 0.286 0.2177 1>
      Day Color=<rgba 0.3594 0.3594 0.3594 1>
      Evening Color=<rgba 0.3073 0.286 0.2177 1>
     )
     Float=&FloatFromCurve(FloatCurve=DataInterface() {propertyValues {Curve {keys [<at 0 1.0483>,<at 0.8725 0>]}}}, CurveIndex=$Engine.System.Age)
    )
   ) {isSetParameters=true, module=SetVariables_#2}
   &ScaleColor(
    Scale Alpha=&FloatFromCurve(FloatCurve=DataInterface() {propertyValues {Curve {keys [{arriveTangent=-3.8864, interpMode=RCIM_Cubic, time=0.3, value=1}, <at 1 0 RCIM_Cubic>]}}})
    Scale RGB=&Multiply_Vector(
     A=&MakeVectorFromLinearColorRGB(Color=&ColorFromCurve(ColorCurve=DataInterface() {propertyValues {
         AlphaCurve {keys [<at 0 1>]}
         BlueCurve {keys [<at 0.017 0.7604>,<at 0.6986 0.4167>]}
         GreenCurve {keys [<at 0.017 0.7604>,<at 0.6986 0.4167>]}
         RedCurve {keys [<at 0.017 0.7604>,<at 0.6986 0.4167>]}
        }}))
     B=&MakeVectorFromLinearColorRGB(Color=&FN_Select_Color_from_Time_Of_Day(
       Use Manual Time of Day Thresholds=true
       Morning=<rgba 0.493 0.4792 0.7812 1>
       Day=<rgba 1 1 1 1>
       Evening=<rgba 0.5469 0.5284 0.7865 1>
       Night=<rgba 0.3581 0.3594 0.7812 1>
       Morning Threshold=9
       Day Threshold=17
       Evening Threshold=21
      ))
    )
    ScaleA=true
    ScaleRGB=true
   )
   &FN_UpdateAnimTrail() {module=FN_UpdateAnimTrail001}
   &SolveForcesAndVelocity(Clamp Velocity=false, Limit Acceleration=false)
   /SetVariables(Particles.SmallPos:NiagaraPosition=&AddVectorToPosition(Position=$Particles.Position, Vector=<0 0 0>)) {isSetParameters=true, module=SetVariables_#3}
   /SetVariables(Particles.GlowSmallWidth=40, Particles.MetalWidth=-21, Particles.RibbonWidthGlow=5) {isSetParameters=true, module=SetVariables_#4}
   &DynamicMaterialParameters(Module.Index0Float0Write=true, Module.Index0Float1Write=true, Module.Index0Float2Write=true, Module.Index0Float3Write=true)
  ]
  rendererClasses [&NiagaraRibbonRendererProperties, &NiagaraSpriteRendererProperties]
  renderers [
   {rendererClass=&NiagaraRibbonRendererProperties}
   {rendererClass=&NiagaraRibbonRendererProperties, rendererIndex=1}
   {rendererClass=&NiagaraRibbonRendererProperties, rendererIndex=2}
   {rendererClass=&NiagaraSpriteRendererProperties, rendererIndex=3}
  ]
 }
 Supporting_Trail {
  emitterUpdateScript [
   &EmitterState()
   &FN_SpawnAnimTrail(Spawn Interval=0) {module=FN_SpawnAnimTrail001}
  ]
  particleSpawnScript [
   &FN_InitializeAnimTrailBoneIndecies(First Bone or Socket Index=2, Second Bone or Socket Index=3)
   &FN_InitializeAnimTrail(AnimTrailWidth=0.3235, Color=<rgba 1.2656 3 1.6082 1>, Lifetime=0.2)
  ]
  particleUpdateScript [
   /SetVariables(
    Particles.GlowColor=&FNTimeOfDayColor(
     Night Color=&ColorFromCurve(ColorCurve=DataInterface() {propertyValues {
        AlphaCurve {keys [{arriveTangent=1.0264, interpMode=RCIM_Cubic, time=0, value=0.5}, <at 0.8901 0>]}
        BlueCurve {keys [<at 0.4889 0.1117 RCIM_Cubic>]}
        GreenCurve {keys [<at 0.4889 5 RCIM_Cubic>]}
        RedCurve {keys [<at 0.4889 0 RCIM_Cubic>]}
       }})
     Morning Color=&ColorFromCurve(ColorCurve=DataInterface() {propertyValues {
        AlphaCurve {keys [{arriveTangent=1.0264, interpMode=RCIM_Cubic, time=0, value=0.687}, <at 0.907 0>]}
        BlueCurve {keys [<at 0.4889 0.1117 RCIM_Cubic>]}
        GreenCurve {keys [<at 0.4889 5 RCIM_Cubic>]}
        RedCurve {keys [<at 0.4889 0 RCIM_Cubic>]}
       }})
     Day Color=&ColorFromCurve(ColorCurve=DataInterface() {propertyValues {
        AlphaCurve {keys [{arriveTangent=1.0264, interpMode=RCIM_Cubic, time=0, value=0.687}, <at 0.557 0>]}
        BlueCurve {keys [<at 0.4889 0.1117 RCIM_Cubic>]}
        GreenCurve {keys [<at 0.4889 5 RCIM_Cubic>]}
        RedCurve {keys [<at 0.4889 0 RCIM_Cubic>]}
       }})
     Evening Color=&ColorFromCurve(ColorCurve=DataInterface() {propertyValues {
        AlphaCurve {keys [{arriveTangent=1.0264, interpMode=RCIM_Cubic, time=0, value=0.687}, <at 0.907 0>]}
        BlueCurve {keys [<at 0.4889 0.1117 RCIM_Cubic>]}
        GreenCurve {keys [<at 0.4889 5 RCIM_Cubic>]}
        RedCurve {keys [<at 0.4889 0 RCIM_Cubic>]}
       }})
    )
    Particles.MetalColor:LinearColor=<rgba 0.5312 0.5312 0.5312 0.5739>
   ) {isSetParameters=true, module=SetVariables_#2}
   &ScaleColor(
    Scale Alpha=&FloatFromCurve(FloatCurve=DataInterface() {propertyValues {Curve {keys [{arriveTangent=-3.8864, interpMode=RCIM_Cubic, time=0.3, value=1}, <at 1 0 RCIM_Cubic>]}}})
    Scale RGB=&Multiply_Vector(
     A=&MakeVectorFromLinearColorRGB(Color=&ColorFromCurve(ColorCurve=DataInterface() {propertyValues {
         AlphaCurve {keys [<at 0 1>]}
         BlueCurve {keys [<at 0.017 0.7604>,<at 0.6986 0.4167>]}
         GreenCurve {keys [<at 0.017 0.7604>,<at 0.6986 0.4167>]}
         RedCurve {keys [<at 0.017 0.7604>,<at 0.6986 0.4167>]}
        }}))
     B=&MakeVectorFromLinearColorRGB(Color=&FN_Select_Color_from_Time_Of_Day(
       Use Manual Time of Day Thresholds=true
       Morning=<rgba 0.493 0.4792 0.7812 1>
       Day=<rgba 1 1 1 1>
       Evening=<rgba 0.5469 0.5284 0.7865 1>
       Night=<rgba 0.3581 0.3594 0.7812 1>
       Morning Threshold=9
       Day Threshold=17
       Evening Threshold=21
      ))
    )
    ScaleA=true
    ScaleRGB=true
   )
   &FN_UpdateAnimTrail() {module=FN_UpdateAnimTrail001}
   &SolveForcesAndVelocity(Clamp Velocity=false, Limit Acceleration=false)
   /SetVariables(Particles.SmallPos:NiagaraPosition=&AddVectorToPosition(Position=$Particles.Position, Vector=<0 0 0>)) {isSetParameters=true, module=SetVariables_#3}
   /SetVariables(Particles.GlowSmallWidth=9, Particles.MetalWidth=12.3104, Particles.RibbonWidthGlow=3.9456) {isSetParameters=true, module=SetVariables_#4}
  ]
  rendererClasses [&NiagaraRibbonRendererProperties]
  renderers [
   {rendererClass=&NiagaraRibbonRendererProperties}
   {rendererClass=&NiagaraRibbonRendererProperties, rendererIndex=1}
  ]
 }
}
summary {
 systemName=NS_Pickaxe_FearCatch_Swing
 emitters [
  {emitterName=MainTrail, rendererClasses [&NiagaraRibbonRendererProperties, &NiagaraSpriteRendererProperties]}
  {emitterName=Supporting_Trail, rendererClasses [&NiagaraRibbonRendererProperties]}
 ]
 userVariables [User.RibbonIDOffset {type=NiagaraInt32}]
}
